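# Translucent UI material that draws the Locker favorite icon in two tintable colors, pink by default.
# A masks texture is thresholded by a SmoothStep around an adjustable edge and margin, in the manner of a distance-field shape.
# The red channel blends between the outer and inner colors and their alphas, while the green channel masks that blended alpha to form the silhouette.
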Material /Locker/Art/M_UI_Locker_Favorite
graph {
 properties {
  bUsedWithStaticMesh=false
  blendMode=BLEND_Translucent
  materialDomain=MD_UI
  pixelDepthOffsetMode=PDOM_Legacy
  usedShadingModels=Default Lit
 }
 nodes [
  n1 = VectorParameter() {group=Colors, parameterName=OuterRGBA, defaultValue=<rgba 1 0.3185 0.5841 1>}
  n2 = VectorParameter() {group=Colors, parameterName=InnerRGBA, defaultValue=<rgba 0.7835 0.0086 0.1845 1>}
  n3 = ScalarParameter() {defaultValue=0.1875, group=Shape, parameterName=SSEdge, sliderMax=1}
  n4 = ScalarParameter() {defaultValue=0.125, group=Shape, parameterName=SSMargin, sliderMax=1}
  n5 = SmoothStep(
   Min=Add(A=@n3, B=Multiply(A=@n4) {constB=-1})
   Max=Add(A=@n3, B=@n4)
   Value=TextureSample(
    UVs=TextureCoordinate()
    Tex=TextureObjectParameter() {parameterName=Texture, samplerType=SAMPLERTYPE_Masks, texture=&/T_UI_Locker_Favorite}
   ) {samplerSource=SSM_Clamp_WorldGroupSettings}.RGB
  )
  n6 = OneMinus(None=Saturate(None=ComponentMask(None=@n5) {r=true}))
  LinearInterpolate(A=@n1.RGB, B=@n2.RGB, Alpha=@n6)
  LinearInterpolate(B=LinearInterpolate(A=@n1.A, B=@n2.A, Alpha=@n6), Alpha=Saturate(None=ComponentMask(None=@n5) {g=true}))
 ]
}
